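# M_PopBalloon is a Default Lit material for the Pop-A-Loon quest interactable in Figment Season 8, evidently for the balloon's surface.
# It exposes a BalloonColor vector parameter, red by default, so material instances can recolor it.
# Two scalar constants are the only other nodes, with no textures or math; the file does not show which material inputs the nodes feed.
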Material /Figment_S08/Athena/Items/QuestInteractables/PopALoon/Materials/M_PopBalloon
graph {
 properties {pixelDepthOffsetMode=PDOM_Legacy, usedShadingModels=Default Lit}
 nodes [
  Constant() {r=0.05}
  Constant()
  VectorParameter() {parameterName=BalloonColor, defaultValue=<rgba 1 0 0 0>}
 ]
}
